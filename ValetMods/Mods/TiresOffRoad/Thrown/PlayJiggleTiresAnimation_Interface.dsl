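; Blueprint interface for the thrown TiresOffRoad mod that lets callers trigger a tire jiggle animation.
; It declares one function, PlayJiggleTiresAnimation, taking double-precision StartFrame and EndFrame values that bound the frame range to play.
; The file holds only the declaration; playback logic lives in whatever implements the interface.
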(fn PlayJiggleTiresAnimation
  (params (StartFrame :type Float%20%28double-precision%29) (EndFrame :type Float%20%28double-precision%29)))
